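# Spot ambient sound cue for a looping chandelier brazier fire in the game world.
# A mixer layers two looping waves, a small statue oil fire boosted above a close mono coliseum brazier fire.
# It uses small chandelier fire attenuation, the Athena ambient class and a default reverb submix send.
# A concurrency override caps it at two instances with distance-based volume scaling.
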SoundCue /Game/Sounds/Helios/Environment_Actors/Brazier/Spot_Brazier_Chandelier_Fire_Loop_Cue
root=Mixer_0
nodes [
 {id=Mixer_0, children [WavePlayer_2,WavePlayer_1], properties {inputVolume [2.5,1]}}
 {
  id=WavePlayer_1
  properties {bLooping=true, soundWaveAssetPtr=/Game/Sounds/Artemis/Environment_Actors/ColiseumBrazierFire/Spot_Coliseum_Brazier_Fire_Close_Mono}
 }
 {
  id=WavePlayer_2
  properties {bLooping=true, soundWaveAssetPtr=&/Spot_Statue_Oil_Fire_Small_Loop}
 }
]
properties {
 attenuationSettings=&/ATT_Fire_Brazier_Chandelier_Small
 bOverrideConcurrency=true
 maxDistance=1350
 soundClassObject=&Fort_Ambient_Athena
 volumeMultiplier=0.501
 concurrencyOverrides {maxCount=2, volumeScaleMode=Distance}
 soundSubmixSends [{minSendLevel=1, soundSubmix=&Fortnite_Reverb_Default}]
}
