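# Looping ambient fire sound cue for a larger brazier, routed to the Athena ambient sound class.
# It mixes a randomly picked urn fire loop at half volume with close mono and stereo Coliseum brazier loops, each with its own attenuation node for distance-based layering.
# A modulator slightly varies the pitch of the whole mix.
# The cue uses a shared focus-named attenuation asset and overrides concurrency to cap instances per owner.
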SoundCue /Game/Sounds/Helios/Environment_Actors/Brazier/Spot_Brazier_Fire_Loop_Larger_Cue
root=Modulator_0
nodes [
 {
  id=Attenuation_0
  children [WavePlayer_1]
  properties {attenuationSettings=/Game/Sounds/Asteria/Environment_Actors/Brazier_LightSource/Spot_Bazier_LightSource_Fire_Close_Att}
 }
 {
  id=Attenuation_1
  children [WavePlayer_2]
  properties {attenuationSettings=/Game/Sounds/Asteria/Environment_Actors/Brazier_LightSource/Spot_Bazier_LightSource_Fire_Att}
 }
 {id=Mixer_0, children [Random_1,Mixer_1], properties {inputVolume [0.5,1]}}
 {id=Mixer_1, children [Attenuation_0, Attenuation_1], properties {inputVolume [1,1]}}
 {id=Modulator_0, children [Mixer_0], properties {pitchMax=1.1, pitchMin=0.9}}
 {id=Random_1, children [WavePlayer_0,WavePlayer_4,WavePlayer_5], properties {weights [1,1,1]}}
 {id=WavePlayer_0, properties {bLooping=true, soundWaveAssetPtr=/Game/Sounds/Trans_AmbientLoops/Fire/Fire_Urn_01}}
 {
  id=WavePlayer_1
  properties {bLooping=true, soundWaveAssetPtr=/Game/Sounds/Artemis/Environment_Actors/ColiseumBrazierFire/Spot_Coliseum_Brazier_Fire_Close_Mono}
 }
 {
  id=WavePlayer_2
  properties {bLooping=true, soundWaveAssetPtr=/Game/Sounds/Artemis/Environment_Actors/ColiseumBrazierFire/Spot_Coliseum_Brazier_Fire_Stereo}
 }
 {id=WavePlayer_4, properties {bLooping=true, soundWaveAssetPtr=/Game/Sounds/Trans_AmbientLoops/Fire/Fire_Urn_02}}
 {id=WavePlayer_5, properties {bLooping=true, soundWaveAssetPtr=/Game/Sounds/Trans_AmbientLoops/Fire/Fire_Urn_03}}
]
properties {
 attenuationSettings=&/Spot_Brazier_LightSource_Fire_Bigger_With_Focus_Att
 bOverrideConcurrency=true
 maxDistance=1850
 soundClassObject=&Fort_Ambient_Athena
 subtitlePriority=5000
 volumeMultiplier=0.702
 attenuationOverrides {
  bEnableFocusInterpolation=true
  bEnableListenerFocus=true
  bEnableOcclusion=true
  dBAttenuationAtMax=-40
  falloffDistance=512
  focusAzimuth=15
  focusVolumeAttenuation=1.5
  nonFocusAzimuth=45
  nonFocusVolumeAttenuation=0.5
  occlusionInterpolationTime=1
  occlusionLowPassFilterFrequency=10000
  occlusionTraceChannel=ECC_GameTraceChannel11
  occlusionVolumeAttenuation=0.5
  stereoSpread=512
  attenuationShapeExtents=<256 64.5942 90>
 }
 concurrencyOverrides {bLimitToOwner=true, maxCount=3}
}
