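# Looping ambient Sound Cue for a ghostly green brimstone torch or brazier fire.
# A mixer layers the looping torch wave over a quieter coliseum brazier fire, then a modulator randomizes pitch slightly upward.
# Two nested looping envelopers of different lengths add slow pitch and volume drift to disguise repetition.
# It uses small positional fire attenuation, a tight concurrency override and a send to the exterior ambient bus.
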SoundCue /Game/Sounds/Helios/Environment_Actors/Brazier/Spot_Brimstone_Fire_Ghostly_Green_Torch_Loop_Cue
root=Enveloper_3
nodes [
 {
  id=Enveloper_0
  children [Modulator_1]
  properties {
   bLoop=true
   bLoopIndefinitely=true
   loopEnd=30
   pitchCurve {editorCurveData {keys [
      {arriveTangent=-0.0013, interpMode=RCIM_Cubic, leaveTangent=-0.0013, time=0}
      <at 15 0.98 RCIM_Cubic>
      {arriveTangent=0.0013, interpMode=RCIM_Cubic, leaveTangent=0.0013, time=30}
     ]}}
   volumeCurve {editorCurveData {keys [<at 0 1>]}}
  }
 }
 {
  id=Enveloper_3
  children [Enveloper_0]
  properties {
   bLoop=true
   bLoopIndefinitely=true
   loopEnd=20
   pitchCurve {editorCurveData {keys [{time=0}]}}
   volumeCurve {editorCurveData {keys [
      {arriveTangent=-0.01, interpMode=RCIM_Cubic, leaveTangent=-0.01, time=0, value=1}
      <at 10 0.9 RCIM_Cubic>
      {arriveTangent=0.01, interpMode=RCIM_Cubic, leaveTangent=0.01, time=20, value=1}
     ]}}
  }
 }
 {id=Mixer_0, children [WavePlayer_4,WavePlayer_1], properties {inputVolume [1.25,0.6]}}
 {id=Modulator_1, children [Mixer_0], properties {pitchMax=1.2, pitchMin=1}}
 {
  id=WavePlayer_1
  properties {bLooping=true, soundWaveAssetPtr=/Game/Sounds/Artemis/Environment_Actors/ColiseumBrazierFire/Spot_Coliseum_Brazier_Fire_Stereo}
 }
 {
  id=WavePlayer_4
  properties {bLooping=true, soundWaveAssetPtr=&/Spot_Brimstone_Fire_Ghostly_Green_Torch_Loop}
 }
]
properties {
 attenuationSettings=/Game/Sounds/Apollo/Biomes/UFO_Crash_Sites/FireLoops/Fire_Smallest_Close_MorePositional_Attenuation
 bBypassVolumeScaleForPriority=true
 bOverrideConcurrency=true
 maxDistance=860
 priority=0
 soundClassObject=/Game/Sounds/FortSoundClassesAndModes/Fort_Ambient_NonLoop
 volumeMultiplier=0.4
 attenuationOverrides {attenuationShape=Cone, attenuationShapeExtents {x=171.3182}}
 busSends [{audioBus=null, minSendLevel=1, sendLevel=1, soundSourceBus=&AmbientAudio_Exterior_SourceBus}]
 concurrencyOverrides {maxCount=3, voiceStealReleaseTime=3}
 soundSubmixSends [{minSendLevel=0, soundSubmix=null}]
}
