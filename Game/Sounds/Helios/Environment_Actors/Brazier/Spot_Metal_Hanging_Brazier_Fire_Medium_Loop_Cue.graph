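# Looping ambient Sound Cue for a medium metal hanging brazier fire, built on the small brazier fire wave.
# The looping wave passes through a random pitch Modulator, then two nested, endlessly looping Envelopers.
# The inner slowly dips pitch and the outer dips volume on different periods, so their combined drift varies over time.
# It uses the shared large-brazier heavy-focus attenuation asset and its own per-owner concurrency limit.
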SoundCue /Game/Sounds/Helios/Environment_Actors/Brazier/Spot_Metal_Hanging_Brazier_Fire_Medium_Loop_Cue
root=Enveloper_3
nodes [
 {
  id=Enveloper_0
  children [Modulator_0]
  properties {
   bLoop=true
   bLoopIndefinitely=true
   loopEnd=30
   pitchCurve {editorCurveData {keys [
      {arriveTangent=-0.0013, interpMode=RCIM_Cubic, leaveTangent=-0.0013, time=0}
      <at 15 0.98 RCIM_Cubic>
      {arriveTangent=0.0013, interpMode=RCIM_Cubic, leaveTangent=0.0013, time=30}
     ]}}
   volumeCurve {editorCurveData {keys [<at 0 1>]}}
  }
 }
 {
  id=Enveloper_3
  children [Enveloper_0]
  properties {
   bLoop=true
   bLoopIndefinitely=true
   loopEnd=20
   pitchCurve {editorCurveData {keys [{time=0}]}}
   volumeCurve {editorCurveData {keys [
      {arriveTangent=-0.01, interpMode=RCIM_Cubic, leaveTangent=-0.01, time=0, value=1}
      <at 10 0.9 RCIM_Cubic>
      {arriveTangent=0.01, interpMode=RCIM_Cubic, leaveTangent=0.01, time=20, value=1}
     ]}}
  }
 }
 {id=Modulator_0, children [WavePlayer_2], properties {pitchMax=1.1, pitchMin=0.9}}
 {
  id=WavePlayer_2
  properties {bLooping=true, soundWaveAssetPtr=&/Spot_Metal_Brazier_Fire_Small_01}
 }
]
properties {
 attenuationSettings=&/Spot_Brazier_Large_With_Heavy_Focus_Att
 bOverrideConcurrency=true
 maxDistance=1850
 soundClassObject=&Fort_Ambient_Athena
 subtitlePriority=5000
 volumeMultiplier=0.501
 attenuationOverrides {
  bEnableFocusInterpolation=true
  bEnableListenerFocus=true
  bEnableOcclusion=true
  dBAttenuationAtMax=-40
  falloffDistance=512
  focusAzimuth=15
  focusVolumeAttenuation=1.5
  nonFocusAzimuth=45
  nonFocusVolumeAttenuation=0.5
  occlusionInterpolationTime=1
  occlusionLowPassFilterFrequency=10000
  occlusionTraceChannel=ECC_GameTraceChannel11
  occlusionVolumeAttenuation=0.5
  stereoSpread=512
  attenuationShapeExtents=<256 64.5942 90>
 }
 concurrencyOverrides {bLimitToOwner=true, maxCount=3}
}
